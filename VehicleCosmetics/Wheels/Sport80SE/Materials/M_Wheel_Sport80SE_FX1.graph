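# Unlit masked material for the Sport80SE cosmetic wheel's FX1 part, drawing an emissive infinity-style effect with tintable custom and background colors.
# A mask texture feeds the infinity offset function borrowed from ClockGear_SE, whose clamped red and green results are added to rerouted values.
# Those sums drive two color blends toward a scaled emissive color, and vertex color blue selects between the blends.
# Only EmissiveColor is wired into the material attributes; OpacityMask is unconnected despite the masked blend mode.
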Material /VehicleCosmetics/Wheels/Sport80SE/Materials/M_Wheel_Sport80SE_FX1
graph {
 properties {
  bAutomaticallySetUsageInEditor=false
  bUseMaterialAttributes=true
  bUsedWithMeshDeformer=true
  bUsedWithSkeletalMesh=true
  blendMode=BLEND_Masked
  pixelDepthOffsetMode=PDOM_Legacy
  previewMesh=/VehicleCosmetics/Wheels/Sport80SE/Meshes/Wheel_Sport80SE
  shadingModel=MSM_Unlit
  usedShadingModels=Unlit
 }
 nodes [
  n1 = NamedRerouteUsage()
  n2 = Multiply(A=ScalarParameter() {defaultValue=1, group=0_SpecialEffects, parameterName=EmissiveColor}, B=NamedRerouteUsage())
  n3 = TextureObjectParameter() {group=0_SpecialEffects, parameterName=Mask, texture=/VehicleCosmetics/Wheels/Sport80SE/Textures/T_Wheel_Sport80SE_FX1_RGB}
  n4 = Clamp(None=MaterialFunctionCall!(
    Texture=@n3
    Offset=ScalarParameter() {defaultValue=-1, parameterName=InfinityOffset}
    UVs=TextureCoordinate()
    RotateUVs=StaticBool()
    GradientFalloff
    GradientDarkness=Constant() {r=0.001}
   ) {materialFunction=/VehicleCosmetics/Wheels/ClockGear_SE/Materials/MF_Infinity_OffsetFunction}.Result)
  n5 = Add(A=ComponentMask(None=@n4) {g=true}, B=NamedRerouteUsage())
  n6 = NamedRerouteUsage()
  n7 = Add(A=ComponentMask(None=@n4) {r=true}, B=NamedRerouteUsage())
  MakeMaterialAttributes!(
   BaseColor
   Metallic
   Specular
   Roughness
   Anisotropy
   EmissiveColor=LinearInterpolate(A=LinearInterpolate(A=@n1, B=@n2, Alpha=@n5), B=LinearInterpolate(A=@n6, B=@n2, Alpha=@n7), Alpha=VertexColor().B)
   Opacity
   OpacityMask
   Normal
   Tangent
   WorldPositionOffset
   SubsurfaceColor
   ClearCoat
   ClearCoatRoughness
   AmbientOcclusion
   Refraction
   CustomizedUVs_0
   CustomizedUVs_1
   CustomizedUVs_2
   CustomizedUVs_3
   CustomizedUVs_4
   CustomizedUVs_5
   CustomizedUVs_6
   CustomizedUVs_7
   PixelDepthOffset
   ShadingModel
   Displacement
  )
  n8 = TextureSample(Tex=@n3) {samplerType=SAMPLERTYPE_LinearColor, texture=&T_Wheel_Sport80SE_FX1_RGB}
  NamedRerouteDeclaration(None=@n8.R) {name=MaskR, nodeColor=<rgba 1 0.7969 0 1>}
  NamedRerouteDeclaration(None=@n8.G) {name=MaskG, nodeColor=<rgba 0.4766 0 1 1>}
  n9 = VectorParameter() {group=0_SpecialEffects, parameterName=CustomColor, sortPriority=0, defaultValue=<rgba 0.8963 0.1812 0.5972 1>}
  NamedRerouteDeclaration(None=@n9.RGB) {name=CustomColor_output, nodeColor=<rgba 0 0.8125 1 1>}
  NamedRerouteDeclaration(None=VectorParameter() {group=0_SpecialEffects, parameterName=BackgroundColor, defaultValue=<rgba 0 0 0 1>}.RGB) {name=BackgroundColor_Output, nodeColor=<rgba 1 0.8672 0 1>}
 ]
}
