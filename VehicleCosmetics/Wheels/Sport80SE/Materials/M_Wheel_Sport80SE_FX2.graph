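# Translucent, lit effects material for the Sport80SE vehicle wheel, adding an animated glow tinted by AccentColor.
# The FX2 mask's red and green channels, plus a normal-map Fresnel, build opacity and blend the accent toward a brighter, whitened tint, feeding both base color and emissive.
# Time-driven named reroutes animate the mask: a sine radial wave from UV1 with speed and opacity parameters, and a slower absolute-sine pulse.
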Material /VehicleCosmetics/Wheels/Sport80SE/Materials/M_Wheel_Sport80SE_FX2
graph {
 properties {
  bAutomaticallySetUsageInEditor=false
  bUseMaterialAttributes=true
  bUsedWithMeshDeformer=true
  bUsedWithSkeletalMesh=true
  blendMode=BLEND_Translucent
  pixelDepthOffsetMode=PDOM_Legacy
  usedShadingModels=Default Lit
 }
 nodes [
  n1 = TextureSampleParameter2D() {
   group=0_SpecialEffects
   parameterName=FX2_Mask
   sortPriority=0
   texture=/VehicleCosmetics/Wheels/Sport80SE/Textures/T_Wheel_Sport80SE_FX2_RGB
  }
  n2 = LinearInterpolate(A=Multiply(B=@n1.R) {constA=0.2, constB=0.5}, Alpha=Multiply(A=NamedRerouteUsage(), B=@n1.R)) {constA=1}
  n3 = Multiply(B=@n1.G) {constA=0.1}
  n4 = TextureSampleParameter2D() {
   group=0_SpecialEffects
   parameterName=FX2_Normal
   samplerType=SAMPLERTYPE_Normal
   sortPriority=2
   texture=/VehicleCosmetics/Wheels/Sport80SE/Textures/T_Wheel_Sport80SE_FX2_N
  }
  n5 = Fresnel(Normal=Transform(None=@n4.RGB)) {exponent=1}
  n6 = Multiply(A=NamedRerouteUsage(), B=Multiply(A=@n1.G, B=LinearInterpolate(Alpha=@n5) {constA=0.2}))
  n7 = VectorParameter() {group=0_SpecialEffects, parameterName=AccentColor, sortPriority=1, defaultValue=<rgba 0.6724 0.3968 0.0343 1>}
  n8 = Constant3Vector() {constant=<rgba 1 1 1 0>}
  n9 = LinearInterpolate(A=@n7.RGB, B=Multiply(A=LinearInterpolate(A=@n7.RGB, B=@n8) {constAlpha=0.6}) {constB=1.5}, Alpha=@n2)
  MakeMaterialAttributes!(
   BaseColor=@n9
   Metallic
   Specular
   Roughness
   Anisotropy
   EmissiveColor=@n9
   Opacity=Add(A=@n2, B=Add(A=@n3, B=@n6) {constB=0.1})
   OpacityMask
   Normal
   Tangent
   WorldPositionOffset
   SubsurfaceColor
   ClearCoat
   ClearCoatRoughness
   AmbientOcclusion
   Refraction
   CustomizedUVs_0
   CustomizedUVs_1
   CustomizedUVs_2
   CustomizedUVs_3
   CustomizedUVs_4
   CustomizedUVs_5
   CustomizedUVs_6
   CustomizedUVs_7
   PixelDepthOffset
   ShadingModel
   Displacement
  )
  n10 = Multiply(A=Subtract(A=TextureCoordinate() {coordinateIndex=1}) {constB=0.5}) {constB=2}
  n11 = Time()
  n12 = Add(
   A=Length(None=ComponentMask(None=@n10) {g=true, r=true}) {is=class MaterialExpressionLength}
   B=Multiply(A=@n11, B=ScalarParameter() {defaultValue=-0.3, group=0_SpecialEffects, parameterName=PulseSpeed, sortPriority=3})
  )
  n13 = Multiply(A=SmoothStep(Value=Sine(None=@n12)) {constMax=6}) {constB=12}
  n14 = ScalarParameter() {defaultValue=1, group=0_SpecialEffects, parameterName=RadialWaveOpacity, sortPriority=4}
  NamedRerouteDeclaration(None=Clamp(None=Multiply(A=@n13, B=@n14))) {name=Radial wave, nodeColor=<rgba 0.3828 1 0 1>}
  n15 = Abs(None=Sine(None=Multiply(A=@n11) {constB=0.2}))
  NamedRerouteDeclaration(None=Clamp(None=Divide(A=@n15))) {name=PulseFX, nodeColor=<rgba 0 1 0.5078 1>}
 ]
}
